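; HUD view model that displays the player pawn's basic rocket ammo as "count / max" text.
; On initialize it gets the pawn's secondary pilot component, with a Delay if that is missing, caches it, binds to its ammo-count-changed event and pushes the current count.
; Each update formats the count with the component's max into the generic dynamic info display text.
; Uninitialize unbinds the event if the cached component is valid.
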(event EventInitialize
  (params (Pawn :type Pawn%20Object%20Reference))
  (|CallParentFunction Pawn)
  (Utilities|IsValid Pawn
    (:Is%20Valid
      (bind _returnvalue (Actor|GetComponentbyClass Pawn "/ToyFoxGameplay/Gameplay/PawnComp/B_ToyFoxPilotPawnComponent_Secondary.B_ToyFoxPilotPawnComponent_Secondary_C"))
      (Utilities|IsValid _returnvalue
        (:Is%20Valid
          _)
        (:Is%20Not%20Valid
          (Utilities|FlowControl|Delay Pawn)))
      (bind _output_get (Variables|Default|SetSecondaryPilotPawnComponent _returnvalue))
      (Default|BindEventtoOnBasicRocketAmmoCountChanged _output_get (EventDispatchers|CreateEvent :!event Custom|UpdateAmmo))
      (CallFunction|UpdateAmmo (Rocket|Basic|GetCurrentBasicRocketAmmoCount _output_get)))))

(event EventUninitialize
  (params (Pawn :type Pawn%20Object%20Reference))
  (bind _secondarypilotpawncomponent (Variables|Default|GetSecondaryPilotPawnComponent))
  (|CallParentFunction Pawn)
  (Utilities|IsValid _secondarypilotpawncomponent
    (:Is%20Valid
      (Default|UnbindEventfromOnBasicRocketAmmoCountChanged _secondarypilotpawncomponent (EventDispatchers|CreateEvent :!event Custom|UpdateAmmo)))))

(event Custom|UpdateAmmo
  (params (AmmoCount :type Integer))
  (Variables|GenericDynamicInfoViewModel|SetDisplayText (Utilities|Text|FormatText "NSLOCTEXT(\"[25AB73E2BDC315DFDBF08CDC1F4F3E6E]\", \"D0FF82694D47F92ED132D7B00866DD7E\", \"{count} / {max}\")" AmmoCount (Rocket|Basic|GetMaxBasicRocketAmmoCount (Variables|Default|GetSecondaryPilotPawnComponent)))))
